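# Ambience sound cue for an underground vent vortex, routed to the indoor ambient sound class.
# A mixer blends two looping cooling-tower waves at boosted input volumes, each through its own overridden attenuation node, under a reduced cue volume.
# One layer is non-spatialized with a logarithmic capsule falloff; the other is spatialized with a wide stereo spread.
# Both use occlusion and listener focus; the mixer's third input is empty and debug is enabled.
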SoundCue /Game/Sounds/Apollo/POI/S12/AMB_Underground_Vent_Vortex_Cue
root=Mixer_0
nodes [
 {
  id=Attenuation_0
  children [WavePlayer_0]
  properties {
   bOverrideAttenuation=true
   attenuationOverrides {
    attenuationShape=Capsule
    bEnableFocusInterpolation=true
    bEnableListenerFocus=true
    bEnableOcclusion=true
    bSpatialize=false
    distanceAlgorithm=Logarithmic
    falloffDistance=1400
    focusReleaseInterpSpeed=2
    nonFocusVolumeAttenuation=0.7
    occlusionLowPassFilterFrequency=1000
    reverbWetLevelMin=0.6429
    attenuationShapeExtents=<5802.106 2761.6406>
   }
  }
 }
 {
  id=Attenuation_1
  children [WavePlayer_1]
  properties {
   attenuationSettings=&Fort_Weapon_Projectile_Medium_Mono_Athena_Att
   bOverrideAttenuation=true
   attenuationOverrides {
    bEnableFocusInterpolation=true
    bEnableListenerFocus=true
    bEnableOcclusion=true
    falloffDistance=4033.6787
    focusAzimuth=55
    focusReleaseInterpSpeed=2
    nonFocusVolumeAttenuation=0.5
    occlusionLowPassFilterFrequency=1000
    stereoSpread=5503.8877
    attenuationShapeExtents {x=3703.1514}
   }
  }
 }
 {id=Mixer_0, children [Attenuation_0, Attenuation_1, null], properties {inputVolume [1.4,1.2,1]}}
 {
  id=WavePlayer_0
  properties {bLooping=true, soundWaveAssetPtr=/Game/Sounds/Fort_Prop_Sounds/Exterior/CoolingTower/PowerPlant_CoolingTower_01}
 }
 {
  id=WavePlayer_1
  properties {bLooping=true, soundWaveAssetPtr=/Game/Sounds/Fort_Prop_Sounds/Exterior/CoolingTower/Cooling_Tower_Medium_Loop_01}
 }
]
properties {
 bDebug=true
 maxDistance=7736.8301
 soundClassObject=&Fort_Ambient_Indoors_Athena
 volumeMultiplier=0.251
 attenuationOverrides {bSpatialize=false, falloffDistance=2000, stereoSpread=1147.9967, attenuationShapeExtents {x=6000}}
}
