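# Sound cue for the ambient wind of an underground vent vortex, reusing a looping cooling tower wind wave.
# The wave player feeds a root attenuation node with a tall, non-spatialized capsule, logarithmic falloff and listener focus that lowers off-focus volume.
# The cue's own attenuation override turns distance attenuation off, leaving falloff to the node.
# It plays through an outdoor-ducking ambient sound class and has its debug flag on.
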SoundCue /Game/Sounds/Apollo/POI/S12/AMB_Underground_Vent_Vortex_Wind_Cue
root=Attenuation_0
nodes [
 {
  id=Attenuation_0
  children [WavePlayer_0]
  properties {
   bOverrideAttenuation=true
   attenuationOverrides {
    attenuationShape=Capsule
    bEnableFocusInterpolation=true
    bEnableListenerFocus=true
    bSpatialize=false
    distanceAlgorithm=Logarithmic
    falloffDistance=1200
    focusReleaseInterpSpeed=2
    lPFFrequencyAtMin=2000
    lPFRadiusMin=2000
    nonFocusVolumeAttenuation=0.7
    occlusionLowPassFilterFrequency=5000
    occlusionTraceChannel=ECC_GameTraceChannel11
    reverbWetLevelMin=0.6429
    attenuationShapeExtents=<8000 800>
   }
  }
 }
 {id=WavePlayer_0, properties {bLooping=true, soundWaveAssetPtr=/Game/Sounds/Fort_Prop_Sounds/Exterior/CoolingTower/CoolingTowerWind_01}}
]
properties {
 bDebug=true
 bOverrideAttenuation=true
 maxDistance=2097152
 soundClassObject=/Game/Sounds/FortSoundClassesAndModes/Fort_Ambient_DuckOutdoors_Athena_PSM
 volumeMultiplier=1.253
 attenuationOverrides {bAttenuate=false, falloffDistance=15000, nonSpatializedRadiusStart=2000, stereoSpread=1000}
}
